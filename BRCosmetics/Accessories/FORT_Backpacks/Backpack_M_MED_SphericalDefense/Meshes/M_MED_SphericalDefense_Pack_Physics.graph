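# Physics asset for the SphericalDefense cosmetic backpack mesh, rigging a dangling tassel for simulation.
# One long capsule covers the pack bone, and a three-bone tassel chain of small capsules, widening toward the tip, hangs from it.
# Angular-limited constraints link each tassel segment to the one above, with narrow first swing and wider second swing and twist.
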PhysicsAsset /BRCosmetics/Accessories/FORT_Backpacks/Backpack_M_MED_SphericalDefense/Meshes/M_MED_SphericalDefense_Pack_Physics
bodies {
 dyn_tassel_01 [{
   shapeType=Capsule
   length=2.4649
   radius=0.505
   shapeName=dyn_tassel_01_capsule
   center=<-0.0182 0.0099 -1.3358>
   rotation=<rot 0.0642 -90 -0.4529>
  }]
 dyn_tassel_02 [{
   shapeType=Capsule
   length=1.3842
   radius=0.7791
   shapeName=dyn_tassel_02_capsule
   center=<0.0554 0.0223 -1.6025>
   rotation=<rot -1.702 -89.3559 -2.3986>
  }]
 dyn_tassel_03 [{
   shapeType=Capsule
   length=1.2538
   radius=1.8458
   shapeName=dyn_tassel_03_capsule
   center=<0.0147 0.1139 -1.6697>
   rotation=<rot 1.9142 167.739 -89.9995>
  }]
 pack [{shapeType=Capsule, length=30.4527, radius=1.763, shapeName=pack_capsule, center=<10.2865 -2.6208 12.0648>, rotation=<rot 1.4091 -90 27.6245>}]
}
constraints [
 {bone1Name=dyn_tassel_02, bone2Name=dyn_tassel_01, swing1LimitDegrees=5, swing2LimitDegrees=45, twistLimitDegrees=45}
 {bone1Name=dyn_tassel_03, bone2Name=dyn_tassel_02, swing1LimitDegrees=8, swing2LimitDegrees=45, twistLimitDegrees=45}
 {bone1Name=dyn_tassel_01, bone2Name=pack, swing1LimitDegrees=5, swing2LimitDegrees=30, twistLimitDegrees=90}
]
