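# Animation Blueprint for the SphericalDefense backpack cosmetic, adjusting the pack's pose to the wearer's movement.
# A cached RigidBody simulation feeds chained additive Modify Bone offsets on the pack bone.
# Update logic derives their alphas from owner spine and thigh reference-transform deltas, using one set for skydiving and another for melee weapons.
# OnWeaponChanged sets IsWeaponChanged, effectively a melee flag, from the player AnimInstance's WeaponCoreAnim.
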AnimBlueprint /BRCosmetics/Accessories/FORT_Backpacks/Backpack_M_MED_SphericalDefense/Meshes/M_MED_SphericalDefense_Pack_AnimBP
target_skeleton=&/M_MED_SphericalDefense_Pack_Skeleton
graphs [
 AnimGraph {nodes [
   n1 = "Transform (Modify) Bone - Bone: pack"!(
    ComponentPose=Local To Component(LocalPose=Use cached pose 'dyn_cache'())
    bAlphaBoolEnabled:bool=NOT Boolean(A=Get bIsSkyDiving())
    Translation="1.784,0.667,0.022"
    Rotation="0,0,0"
    Scale
   ) -> (Pose) {
    alphaInputType=Bool
    rotationMode=BMM_Additive
    rotationSpace=BCS_BoneSpace
    translationMode=BMM_Additive
    alphaBoolBlend {blendOption=Linear}
    alphaScaleBiasClamp {bInterpResult=true, interpSpeedIncreasing=9}
    boneToModify {boneName=pack}
    translation=<1.784 0.667 0.022>
   }
   n2 = "Transform (Modify) Bone - Bone: pack"!(
    ComponentPose=@n1
    Alpha:float=Get RotationRightSpine_X_jump() -> (RotationRightSpine_X_jump:float)
    Translation="0,0,0"
    Rotation="-2.250,-0.729,18.507"
    Scale
   ) -> (Pose) {
    rotationMode=BMM_Additive
    rotationSpace=BCS_BoneSpace
    translationMode=BMM_Additive
    alphaBoolBlend {blendOption=Linear}
    alphaScaleBiasClamp {bInterpResult=true, interpSpeedIncreasing=9}
    boneToModify {boneName=pack}
    rotation=<rot -2.25 -0.729 18.507>
   }
   n3 = "Transform (Modify) Bone - Bone: pack"(
    ComponentPose=@n2
    Alpha=Get RotationRightSpine_X_onGround() -> (RotationRightSpine_X_onGround:float)
    Translation="1.708,-3.980,-2.932"
    Rotation="-0.264,-2.976,16.036"
   ) {
    rotationMode=BMM_Additive
    rotationSpace=BCS_BoneSpace
    translationMode=BMM_Additive
    alphaBoolBlend {blendOption=Linear}
    alphaScaleBiasClamp {bInterpResult=true, interpSpeedIncreasing=9}
    boneToModify {boneName=pack}
    rotation=<rot -0.264 -2.976 16.036>
    translation=<1.708 -3.98 -2.932>
   }
   n4 = "Transform (Modify) Bone - Bone: pack"(
    ComponentPose=@n3
    Alpha=Get RotationLeftSpine_X_onGround() -> (RotationLeftSpine_X_onGround:float)
    Translation="0.000,0.367,-0.185"
    Rotation="-0.000,0.000,5.802"
   ) {
    rotationMode=BMM_Additive
    rotationSpace=BCS_BoneSpace
    translationMode=BMM_Additive
    alphaBoolBlend {blendOption=Linear}
    alphaScaleBiasClamp {bInterpResult=true, interpSpeedIncreasing=9}
    boneToModify {boneName=pack}
    rotation {roll=5.802}
    translation {y=0.367, z=-0.185}
   }
   n5 = "Transform (Modify) Bone - Bone: pack"(
    ComponentPose=@n4
    Alpha=Get Rotation_spine5_z_inSky() -> (Rotation_spine5_z_inSky:float)
    Translation="-0.274,-14.098,-1.560"
    Rotation="0.857,13.754,42.778"
   ) {
    rotationMode=BMM_Additive
    rotationSpace=BCS_BoneSpace
    translationMode=BMM_Additive
    alphaBoolBlend {blendOption=Linear}
    boneToModify {boneName=pack}
    rotation=<rot 0.857 13.754 42.778>
    translation=<-0.274 -14.098 -1.56>
   }
   n6 = "Transform (Modify) Bone - Bone: pack"(
    ComponentPose=@n5
    Alpha=Get Rotation_spine1_z_inSky() -> (Rotation_spine1_z_inSky:float)
    Translation="0,0,0"
    Rotation="-54.067,0.006,-0.007"
   ) {
    rotationMode=BMM_Additive
    rotationSpace=BCS_BoneSpace
    translationMode=BMM_Additive
    alphaBoolBlend {blendOption=Linear}
    boneToModify {boneName=pack}
    rotation=<rot -54.067 0.006 -0.007>
   }
   n7 = "Transform (Modify) Bone - Bone: pack"(
    ComponentPose=@n6
    Alpha=Get Rotation_RightThigh_z_inSky() -> (Rotation_RightThigh_z_inSky:float)
    Translation="-2.206,0.922,0.218"
    Rotation="6.402,-7.126,7.199"
   ) {
    rotationMode=BMM_Additive
    rotationSpace=BCS_BoneSpace
    translationMode=BMM_Additive
    alphaBoolBlend {blendOption=Linear}
    boneToModify {boneName=pack}
    rotation=<rot 6.402 -7.126 7.199>
    translation=<-2.206 0.922 0.218>
   }
   n8 = "Transform (Modify) Bone - Bone: pack"(
    ComponentPose=@n7
    Alpha=Get ExtraRotation_RightThigh_z_inSky() -> (ExtraRotation_RightThigh_z_inSky:float)
    Translation="0.147,-8.004,5.521"
    Rotation="-1.615,3.183,30.876"
   ) {
    rotationMode=BMM_Additive
    rotationSpace=BCS_BoneSpace
    translationMode=BMM_Additive
    alphaBoolBlend {blendOption=Linear}
    alphaScaleBiasClamp {bInterpResult=true, interpSpeedDecreasing=6, interpSpeedIncreasing=30}
    boneToModify {boneName=pack}
    rotation=<rot -1.615 3.183 30.876>
    translation=<0.147 -8.004 5.521>
   }
   n9 = "Transform (Modify) Bone - Bone: pack"!(ComponentPose=@n8, bAlphaBoolEnabled:bool=Get bIsSkyDiving(), Translation="0.637,-0.118,2.545", Rotation="-11.632,0.068,-0.026", Scale) -> (Pose) {
    alphaInputType=Bool
    rotationMode=BMM_Additive
    rotationSpace=BCS_BoneSpace
    translationMode=BMM_Additive
    alphaBoolBlend {blendOption=Linear}
    alphaScaleBiasClamp {bInterpResult=true, interpSpeedIncreasing=9}
    boneToModify {boneName=pack}
    rotation=<rot -11.632 0.068 -0.026>
    translation=<0.637 -0.118 2.545>
   }
   Output Pose(Result=Component To Local(ComponentPose=@n9)) {layerGroup=None}
   n10 = RigidBody(ComponentPose=Local To Component(LocalPose=Input Pose())) {
    bOverrideWorldGravity=true
    overridePhysicsAsset=&/M_MED_SphericalDefense_Pack_Physics
    alphaBoolBlend {blendOption=Linear}
    componentLinearAccScale=<0.2 0.2 0.2>
    componentLinearVelScale=<0.2 0.2 0.2>
    overrideWorldGravity {z=-980}
   }
   Save cached pose 'dyn_cache'(Pose=Component To Local(ComponentPose=@n10)) {cacheName=dyn_cache}
  ]}
 EventGraph {nodes [
   Event BlueprintUpdateAnimation()
   > n1 = Sequence!(execute) -> (then_0, then_1, then_2:Exec)
   n2 = Branch(execute=@n1.then_1, Condition=Get bIsSkyDiving())
   > Set RotationRightSpine_X_onGround(RotationRightSpine_X_onGround:float=0) -> (Output_Get:float)
   > Set RotationLeftSpine_X_onGround(RotationLeftSpine_X_onGround:float=0) -> (Output_Get:float)
   > Set RotationRightSpine_X_jump(RotationRightSpine_X_jump:float=0) -> (Output_Get:float)
   > Set RotationSpineCompintation_X_onGround(RotationSpineCompintation_X_onGround:float=0) -> (Output_Get:float)
   n3 = GetOwnerSkeletalMeshComponent()
   > Set Rotation_spine5_z_inSky(Rotation_spine5_z_inSky:float=GetAlphaFromReferenceTransformDelta(SkeletalMeshComponent=@n3, BoneName=spine_05, BaseName=pelvis, InputRange=(X=-40,Y=-80), TwistAxis="1,0,0")) -> (Output_Get:float)
   > Set Rotation_spine1_z_inSky(Rotation_spine1_z_inSky:float=GetAlphaFromReferenceTransformDelta(SkeletalMeshComponent=@n3, BoneName=spine_01, BaseName=pelvis, InputRange=(X=0,Y=80), TwistAxis="1,0,0")) -> (Output_Get:float)
   > Set Rotation_RightThigh_z_inSky(Rotation_RightThigh_z_inSky:float=GetAlphaFromReferenceTransformDelta(SkeletalMeshComponent=@n3, BoneName=thigh_r, BaseName=spine_05, InputRange=(X=10,Y=-12), TwistAxis="1,0,0")) -> (Output_Get:float)
   > Set ExtraRotation_RightThigh_z_inSky(ExtraRotation_RightThigh_z_inSky:float=GetAlphaFromReferenceTransformDelta(SkeletalMeshComponent=@n3, BoneName=spine_05, BaseName=spine_04, InputRange=(X=-10,Y=-20), TwistAxis="1,0,0")) -> (Output_Get:float)
   Set Rotation_spine5_z_inSky(execute=@n2.else, Rotation_spine5_z_inSky=0)
   > Set Rotation_spine1_z_inSky(Rotation_spine1_z_inSky=0)
   > Set Rotation_RightThigh_z_inSky(Rotation_RightThigh_z_inSky=0)
   n4 = Branch(execute=@n1.then_2, Condition=Get IsWeaponChanged() -> (IsWeaponChanged:bool))
   n5 = GetOwnerSkeletalMeshComponent()
   > Set RotationRightSpine_X_onGround(RotationRightSpine_X_onGround=GetAlphaFromReferenceTransformDelta(SkeletalMeshComponent=@n5, BoneName=spine_05, BaseName=pelvis, InputRange=(X=-7,Y=-10), TwistAxis="1,0,0"))
   > Set RotationLeftSpine_X_onGround(RotationLeftSpine_X_onGround=GetAlphaFromReferenceTransformDelta(SkeletalMeshComponent=@n5, BoneName=spine_05, BaseName=spine_02, DeltaComponentType=EulerX, InputRange=(X=10,Y=30), TwistAxis="0,0,1"))
   > Set RotationRightSpine_X_jump(RotationRightSpine_X_jump=GetAlphaFromReferenceTransformDelta(
     SkeletalMeshComponent=@n5
     BoneName=spine_05
     BaseName=pelvis
     DeltaComponentType=QuaternionTwist
     InputRange=(X=-40,Y=-55)
     TwistAxis="1,0,0"
    ))
   Set RotationRightSpine_X_onGround(execute=@n4.else, RotationRightSpine_X_onGround=0)
   > Set RotationLeftSpine_X_onGround(RotationLeftSpine_X_onGround=0)
   > Set RotationRightSpine_X_jump(RotationRightSpine_X_jump=0)
   > Set RotationSpineCompintation_X_onGround(RotationSpineCompintation_X_onGround=0)
   n6 = Cast To FortPlayerAnimInstance(execute=@n1.then_0, Object=GetAnimInstance(self=GetOwnerSkeletalMeshComponent()))
   Event OnWeaponChanged()
   > Set IsWeaponChanged(IsWeaponChanged:bool=false) -> (Output_Get:bool)
   > n7 = Switch on EFortWeaponCoreAnimation!(execute, Selection:EFortWeaponCoreAnimation Enum=Get WeaponCoreAnim(self=@n6)) -> (
    Melee:Exec
    Pistol:Exec
    Shotgun:Exec
    PaperBlueprint:Exec
    Rifle:Exec
    MeleeOneHand:Exec
    MachinePistol:Exec
    RocketLauncher:Exec
    GrenadeLauncher:Exec
    GoingCommando:Exec
    AssaultRifle:Exec
    TacticalShotgun:Exec
    SniperRifle:Exec
    TrapPlacement:Exec
    ShoulderLauncher:Exec
    AbilityDecoTool:Exec
    Crossbow:Exec
    C4:Exec
    RemoteControl:Exec
    DualWield:Exec
    AR_BullPup:Exec
    AR_ForwardGrip:Exec
    MedPackPaddles:Exec
    SMG_P90:Exec
    AR_DrumGun:Exec
    Consumable_Small:Exec
    Consumable_Large:Exec
    Balloon:Exec
    MountedTurret:Exec
    CreativeTool:Exec
    ExplosiveBow:Exec
    AshtonIndigo:Exec
    AshtonChicago:Exec
    MeleeDualWield:Exec
    Unarmed:Exec
   )
   Set IsWeaponChanged(execute=[@n7.Melee, @n7.MeleeDualWield, @n7.MeleeOneHand], IsWeaponChanged=true)
   TryGetPawnOwner()
   Reroute Node(InputPin:Wildcard) -> (OutputPin:Wildcard)
  ]}
]
variables [
 ExtraRotation_RightThigh_z_inSky {type=number}
 IsWeaponChanged {type=boolean}
 Rotation_RightThigh_z_inSky {type=number}
 Rotation_spine1_z_inSky {type=number}
 Rotation_spine5_z_inSky {type=number}
 RotationLeftSpine_X_onGround {type=number}
 RotationRightSpine_X_jump {type=number}
 RotationRightSpine_X_onGround {type=number}
 RotationSpineCompintation_X_onGround {type=number}
]
